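# Niagara deploy effect for the FloralMane glider, emitting short-lived mesh particle bursts from sockets on the glider's skeletal mesh.
# A random front-to-back value sets socket offset, orientation and scale, spawn group mirrors sides, and ForwardDot/RightDot bend orientation.
# Two delayed waves (the second fixed to group 1) animate scale and material parameters.
# Front-end duplicates, gated by spawn rate, start sooner and resample sockets every update.
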NiagaraSystem /BRCosmetics/Gadgets/Assets/VinderTech_GliderChute/Glider_FloralMane/FX/NS_Glider_FloralMane_Deploy
dependencies {
 usedDataInterfaces [&NiagaraDataInterfaceCurve, &NiagaraDataInterfaceVectorCurve]
 usedDynamicInputs [&Boolean_OrOperation, &SetBoolByFloatComparison]
 usedModules [
  &self:SystemSpawnScript.SystemScriptSource.SystemScriptGraph.NiagaraNodeAssignment_1.SetVariables_#1
  &SystemState
  &Timeline
  &FN_Integrator
  &self:SystemSpawnScript.SystemScriptSource.SystemScriptGraph.NiagaraNodeAssignment_0.SetVariables_#2
 ]
}
emitters {
 Deploy_First {
  emitterUpdateScript [
   &EmitterState(Life Cycle Mode=::NewEnumerator1, Loop Behavior=::NewEnumerator1, Loop Delay=0.3, Loop Duration=0.3, UseLoopDelay=true)
   &SpawnRate(
    SpawnGroup:NiagaraInt32=&RandomRangeInt(Minimum=0, Maximum=1)
    SpawnRate=&FN_FrontEndFloatModifier(Gameplay=160, Front End=0)
    Use Spawn Probability=false
   )
  ]
  particleSpawnScript [
   &FN_SetSpawnGroup(Number of Spawn Groups=2)
   /SetVariables() {isSetParameters=true, module=SetVariables_#3}
   /SetVariables(
    Particles.FrontToBack:NiagaraFloat=&RandomRangeFloat(Minimum=0, Maximum=1)
    Particles.LeanLerp:NiagaraFloat=&Multiply_Float(
     A=$User.RightDot
     B=&Lerp_Float(A=-1, B=1, Alpha=&MakeFloatFromInt(Int=$Particles.SpawnGroup))
    )
   ) {isSetParameters=true, module=SetVariables_#4}
   &InitializeParticle(
    Lifetime Max=0.75
    Lifetime Min=0.75
    Lifetime Mode=::NewEnumerator1
    Mesh Scale Max=<1 2 1.2>
    Mesh Scale Min=<0.8 1.5 1>
    Mesh Scale Mode=::NewEnumerator2
    Sprite Rotation Mode=::NewEnumerator2
    UsePositionOffset=false
    Write Material Random=true
   ) {module=InitializeParticle001}
   &InitialMeshOrientation(
    Orientation Vector=&Multiply_Vector(
     A=&Add_Vector(
      A=&Lerp_Vector(A=<0 4 0>, B=<-2 0 0>, Alpha=$Particles.FrontToBack)
      B=&Lerp_Vector(A=<-1 0.8 0>, B=<-0.75 1 0>, Alpha=$Particles.LeanLerp)
     )
     B=&Multiply_Vector(
      A=&Lerp_Vector(A=<1 1 1>, B=<1 -1 1>, Alpha=&MakeFloatFromInt(Int=$Particles.SpawnGroup))
      B=&Lerp_Vector(A=<1 1 1>, B=<1 2 1>, Alpha=&ClampFloat(Float=$User.ForwardDot))
     )
    )
    Rotation Coordinate Space=::NewEnumerator0
    Use Orientation Vector=true
    Use Rotation Vector=false
   )
   /SetVariables(Particles.Scale=&Multiply_Vector(
     A=&Multiply_Vector(
      A=&VectorFromCurve(
       VectorCurve=DataInterface() {propertyValues {
         XCurve {keys [{arriveTangent=1.7846, interpMode=RCIM_Cubic, leaveTangent=1.7846, tangentMode=RCTM_User, time=0, value=0.4}, <at 1 1 RCIM_Cubic>]}
         YCurve {keys [{arriveTangent=1.7846, interpMode=RCIM_Cubic, leaveTangent=1.7846, tangentMode=RCTM_User, time=0, value=0.4}, <at 1 1 RCIM_Cubic>]}
         ZCurve {keys [{arriveTangent=1.7846, interpMode=RCIM_Cubic, leaveTangent=1.7846, tangentMode=RCTM_User, time=0, value=0.4}, <at 1 1 RCIM_Cubic>]}
        }}
       CurveIndex=$Particles.FrontToBack
       Scale Curve=&Multiply_VectorByFloat(
        Vector=$Particles.Scale
        Float=&Lerp_Float(
         A=0.8
         B=0.8
         Alpha=&ClampFloat(Float=&Max_Float(A:NiagaraFloat=$User.ForwardDot, B:NiagaraFloat=$System.Timeline004.NormalizedTime))
        )
       )
      )
      B=&Lerp_Vector(A=<0.8 0.8 0.8>, B=<0.8 0.8 0.8>, Alpha=$Particles.LeanLerp)
     )
     B=&MakeVector(
      X=&Lerp_Float(
       A=1
       B=1.5
       Alpha=&FloatFromCurve(
        FloatCurve=DataInterface() {propertyValues {Curve {
           defaultValue=0
           keys [
            {arriveTangent=8.0076, interpMode=RCIM_Cubic, leaveTangent=8.0076, tangentMode=RCTM_User, time=0.6, value=0}
            {interpMode=RCIM_Cubic, tangentMode=RCTM_User, time=1, value=1}
           ]
          }}}
        CurveIndex=$Particles.FrontToBack
       )
      )
      Y=&Lerp_Float(A=2, B=1, Alpha=$Particles.FrontToBack)
      Z=1
     )
    )) {isSetParameters=true, module=SetVariables_#5}
   &SkeletalMeshLocation(
    Bone / Socket Index=$Particles.SpawnGroup
    "Inherit Velocity (Scale)"=-1
    MaskSpawnGroup=false
    Mesh Sampling Type=::NewEnumerator5
    Sampled Position Offset=&VectorFromCurve(
     VectorCurve=DataInterface() {propertyValues {
       XCurve {keys [<at 0 50>, {arriveTangent=-382.5781, interpMode=RCIM_Cubic, leaveTangent=-382.5781, tangentMode=RCTM_User, time=1, value=-190}]}
       YCurve {keys [<at 0.33 0 RCIM_Cubic>, {arriveTangent=-92.7914, interpMode=RCIM_Cubic, leaveTangent=-92.7914, tangentMode=RCTM_User, time=1, value=-35}]}
       ZCurve {keys [<at 0 0>,<at 1 10>]}
      }}
     CurveIndex=$Particles.FrontToBack
     Scale Curve=&Lerp_Vector(A=<1 1 1>, B=<1 -1 1>, Alpha=&MakeFloatFromInt(Int=$Particles.SpawnGroup))
    )
    Skeletal Mesh=$User.SkeletalMesh
    Socket Sampling Mode=::NewEnumerator5
    Velocity Sampling=::NewEnumerator0
    VelocitySpeedLimit=false
   ) {module=SkeletalMeshLocation004}
   &AddVelocity(Velocity=<1 0 1>, Velocity Speed Scale=500) {module=AddVelocity001}
  ]
  particleUpdateScript [
   &ParticleState()
   /SetVariables(Particles.Scale=&Multiply_Vector(
     A=$Particles.Initial.Scale
     B=&VectorFromCurve(VectorCurve=DataInterface() {propertyValues {
        XCurve {keys [
          {arriveTangent=4.6994, interpMode=RCIM_Cubic, leaveTangent=4.6994, tangentMode=RCTM_User, time=0, value=0.0795}
          {arriveTangent=0.4589, interpMode=RCIM_Cubic, leaveTangent=0.4589, tangentMode=RCTM_User, time=0.1, value=0.7609}
          <at 0.5 1>
         ]}
        YCurve {keys [{arriveTangent=2.5596, interpMode=RCIM_Cubic, leaveTangent=2.5596, tangentMode=RCTM_User, time=0, value=0.0767}, <at 0.5 1>]}
        ZCurve {keys [
          <at 0 0 RCIM_Cubic>
          {arriveTangent=-2.3619, interpMode=RCIM_Cubic, leaveTangent=-2.3619, tangentMode=RCTM_User, time=0.15, value=0.9567}
          <at 0.9988 0>
         ]}
       }})
    )) {isSetParameters=true, module=SetVariables_#6}
   &InheritSourceMovement(ApplyForce=false, ApplyPosition=true, ApplyPositionFalloff=false)
   &SolveForcesAndVelocity(Clamp Velocity=false, Limit Acceleration=false)
   &DynamicMaterialParameters(
    Index 0 Param 1=&FloatFromCurve(FloatCurve=DataInterface() {propertyValues {Curve {keys [
         {arriveTangent=0.0095, interpMode=RCIM_Cubic, leaveTangent=0.0095, tangentMode=RCTM_User, time=0, value=-1.5}
         {arriveTangent=0.0695, interpMode=RCIM_Cubic, leaveTangent=0.0695, tangentMode=RCTM_User, time=1, value=0}
        ]}}})
    Index 0 Param 2=&FloatFromCurve(FloatCurve=DataInterface() {propertyValues {Curve {keys [<at 0 4>,<at 1 4>]}}})
    Index 0 Param 3=&MakeFloatFromVector(Vector=$Particles.Scale, Channel=::NewEnumerator0)
    Index 0 Param 4=1
    Index 1 Param 1=&FloatFromCurve(FloatCurve=DataInterface() {propertyValues {Curve {keys [<at 0.7 0 RCIM_Cubic>, {arriveTangent=6.4641, interpMode=RCIM_Cubic, leaveTangent=6.4641, tangentMode=RCTM_User, time=1, value=1}]}}})
    Module.Param0WriteEnabled=true
    Module.Param0WriteEnabled1=true
    Module.Param1WriteEnabled=true
    Module.Param1WriteEnabled1=true
    Module.Param2WriteEnabled=true
    Module.Param2WriteEnabled1=true
    Module.Param3WriteEnabled=true
    Module.Param3WriteEnabled1=true
    Write Parameter Index 1=true
   )
   &ScaleColor(
    Scale Alpha=&FloatFromCurve(FloatCurve=DataInterface() {propertyValues {Curve {keys [{interpMode=RCIM_Cubic, tangentMode=RCTM_User, time=0, value=1}, {interpMode=RCIM_Cubic, tangentMode=RCTM_User, time=0.16, value=1}]}}})
    ScaleA=true
    ScaleRGB=true
   )
  ]
  rendererClasses [&NiagaraMeshRendererProperties]
  renderers [{rendererClass=&NiagaraMeshRendererProperties, rendererIndex=0}]
 }
 Deploy_Second {
  emitterUpdateScript [
   &EmitterState(Life Cycle Mode=::NewEnumerator1, Loop Behavior=::NewEnumerator1, Loop Delay=0.8, Loop Duration=0.6, UseLoopDelay=true)
   &SpawnRate(
    SpawnGroup:NiagaraInt32=&RandomRangeInt(Minimum=0, Maximum=1)
    SpawnRate=&FN_FrontEndFloatModifier(Gameplay=160, Front End=0)
    Use Spawn Probability=false
   )
  ]
  particleSpawnScript [
   /SetVariables(Particles.SpawnGroup=1) {isSetParameters=true, module=SetVariables_#3}
   /SetVariables(
    Particles.FrontToBack:NiagaraFloat=&RandomRangeFloat(Minimum=0, Maximum=1)
    Particles.LeanLerp:NiagaraFloat=&Multiply_Float(
     A=$User.RightDot
     B=&Lerp_Float(A=-1, B=1, Alpha=&MakeFloatFromInt(Int=$Particles.SpawnGroup))
    )
   ) {isSetParameters=true, module=SetVariables_#4}
   &InitializeParticle(
    Lifetime Max=0.75
    Lifetime Min=0.75
    Lifetime Mode=::NewEnumerator1
    Mesh Scale Max=<1.5 2 1.2>
    Mesh Scale Min=<1 1.5 1>
    Mesh Scale Mode=::NewEnumerator2
    Sprite Rotation Mode=::NewEnumerator2
    UsePositionOffset=false
    Write Material Random=true
   ) {module=InitializeParticle001}
   &InitialMeshOrientation(
    Orientation Vector=&Multiply_Vector(
     A=&Add_Vector(
      A=&Lerp_Vector(A=<0 4 0>, B=<-2 0 0>, Alpha=$Particles.FrontToBack)
      B=&Lerp_Vector(A=<-1 0.8 0>, B=<-0.75 1 0>, Alpha=$Particles.LeanLerp)
     )
     B=&Multiply_Vector(
      A=&Lerp_Vector(A=<1 1 1>, B=<1 -1 1>, Alpha=&MakeFloatFromInt(Int=$Particles.SpawnGroup))
      B=&Lerp_Vector(A=<1 1 1>, B=<1 2 1>, Alpha=&ClampFloat(Float=$User.ForwardDot))
     )
    )
    Rotation Coordinate Space=::NewEnumerator0
    Use Orientation Vector=true
    Use Rotation Vector=false
   )
   /SetVariables(Particles.Scale=&Multiply_Vector(
     A=&Multiply_Vector(
      A=&VectorFromCurve(
       VectorCurve=DataInterface() {propertyValues {
         XCurve {keys [{arriveTangent=1.7846, interpMode=RCIM_Cubic, leaveTangent=1.7846, tangentMode=RCTM_User, time=0, value=0.4}, <at 1 1 RCIM_Cubic>]}
         YCurve {keys [{arriveTangent=1.7846, interpMode=RCIM_Cubic, leaveTangent=1.7846, tangentMode=RCTM_User, time=0, value=0.4}, <at 1 1 RCIM_Cubic>]}
         ZCurve {keys [{arriveTangent=1.7846, interpMode=RCIM_Cubic, leaveTangent=1.7846, tangentMode=RCTM_User, time=0, value=0.4}, <at 1 1 RCIM_Cubic>]}
        }}
       CurveIndex=$Particles.FrontToBack
       Scale Curve=&Multiply_VectorByFloat(
        Vector=$Particles.Scale
        Float=&Lerp_Float(
         A=0.8
         B=0.8
         Alpha=&ClampFloat(Float=&Max_Float(A:NiagaraFloat=$User.ForwardDot, B:NiagaraFloat=$System.Timeline004.NormalizedTime))
        )
       )
      )
      B=&Lerp_Vector(A=<0.8 0.8 0.8>, B=<0.8 0.8 0.8>, Alpha=$Particles.LeanLerp)
     )
     B=&MakeVector(
      X=&Lerp_Float(
       A=1
       B=1.5
       Alpha=&FloatFromCurve(
        FloatCurve=DataInterface() {propertyValues {Curve {
           defaultValue=0
           keys [
            {arriveTangent=8.0076, interpMode=RCIM_Cubic, leaveTangent=8.0076, tangentMode=RCTM_User, time=0.6, value=0}
            {interpMode=RCIM_Cubic, tangentMode=RCTM_User, time=1, value=1}
           ]
          }}}
        CurveIndex=$Particles.FrontToBack
       )
      )
      Y=&Lerp_Float(A=2, B=1, Alpha=$Particles.FrontToBack)
      Z=1
     )
    )) {isSetParameters=true, module=SetVariables_#5}
   &SkeletalMeshLocation(
    Bone / Socket Index=$Particles.SpawnGroup
    MaskSpawnGroup=false
    Mesh Sampling Type=::NewEnumerator5
    Sampled Position Offset=&VectorFromCurve(
     VectorCurve=DataInterface() {propertyValues {
       XCurve {keys [<at 0 50>, {arriveTangent=-382.5781, interpMode=RCIM_Cubic, leaveTangent=-382.5781, tangentMode=RCTM_User, time=1, value=-190}]}
       YCurve {keys [<at 0.33 0 RCIM_Cubic>, {arriveTangent=-92.7914, interpMode=RCIM_Cubic, leaveTangent=-92.7914, tangentMode=RCTM_User, time=1, value=-35}]}
       ZCurve {keys [<at 0 0>,<at 1 10>]}
      }}
     CurveIndex=$Particles.FrontToBack
     Scale Curve=&Lerp_Vector(A=<1 1 1>, B=<1 -1 1>, Alpha=&MakeFloatFromInt(Int=$Particles.SpawnGroup))
    )
    Skeletal Mesh=$User.SkeletalMesh
    Socket Sampling Mode=::NewEnumerator5
   ) {module=SkeletalMeshLocation004}
   &AddVelocity(Velocity=<-1 0.25 1>, Velocity Speed Scale=500)
  ]
  particleUpdateScript [
   &ParticleState()
   /SetVariables(Particles.Scale=&Multiply_Vector(
     A=$Particles.Initial.Scale
     B=&VectorFromCurve(VectorCurve=DataInterface() {propertyValues {
        XCurve {keys [
          {arriveTangent=4.6994, interpMode=RCIM_Cubic, leaveTangent=4.6994, tangentMode=RCTM_User, time=0, value=0.0795}
          {arriveTangent=0.4589, interpMode=RCIM_Cubic, leaveTangent=0.4589, tangentMode=RCTM_User, time=0.1, value=0.7609}
          <at 0.5 1>
         ]}
        YCurve {keys [{arriveTangent=2.5596, interpMode=RCIM_Cubic, leaveTangent=2.5596, tangentMode=RCTM_User, time=0, value=0.0767}, <at 0.5 1>]}
        ZCurve {keys [
          <at 0 0 RCIM_Cubic>
          {arriveTangent=-2.3619, interpMode=RCIM_Cubic, leaveTangent=-2.3619, tangentMode=RCTM_User, time=0.15, value=0.9567}
          <at 0.9988 0>
         ]}
       }})
    )) {isSetParameters=true, module=SetVariables_#6}
   &InheritSourceMovement(ApplyForce=false, ApplyPosition=true, ApplyPositionFalloff=false)
   &SolveForcesAndVelocity(Clamp Velocity=false, Limit Acceleration=false)
   &DynamicMaterialParameters(
    Index 0 Param 1=&FloatFromCurve(FloatCurve=DataInterface() {propertyValues {Curve {keys [
         {arriveTangent=0.0095, interpMode=RCIM_Cubic, leaveTangent=0.0095, tangentMode=RCTM_User, time=0, value=-1.5}
         {arriveTangent=0.0695, interpMode=RCIM_Cubic, leaveTangent=0.0695, tangentMode=RCTM_User, time=1, value=0}
        ]}}})
    Index 0 Param 2=&FloatFromCurve(FloatCurve=DataInterface() {propertyValues {Curve {keys [<at 0 4>,<at 1 4>]}}})
    Index 0 Param 3=&MakeFloatFromVector(Vector=$Particles.Scale, Channel=::NewEnumerator0)
    Index 0 Param 4=1
    Index 1 Param 1=&FloatFromCurve(FloatCurve=DataInterface() {propertyValues {Curve {keys [<at 0.7 0 RCIM_Cubic>, {arriveTangent=6.4641, interpMode=RCIM_Cubic, leaveTangent=6.4641, tangentMode=RCTM_User, time=1, value=1}]}}})
    Module.Param0WriteEnabled=true
    Module.Param0WriteEnabled1=true
    Module.Param1WriteEnabled=true
    Module.Param1WriteEnabled1=true
    Module.Param2WriteEnabled=true
    Module.Param2WriteEnabled1=true
    Module.Param3WriteEnabled=true
    Module.Param3WriteEnabled1=true
    Write Parameter Index 1=true
   )
   &ScaleColor(
    Scale Alpha=&FloatFromCurve(FloatCurve=DataInterface() {propertyValues {Curve {keys [{interpMode=RCIM_Cubic, tangentMode=RCTM_User, time=0, value=1}, {interpMode=RCIM_Cubic, tangentMode=RCTM_User, time=0.16, value=1}]}}})
    ScaleA=true
    ScaleRGB=true
   )
  ]
  rendererClasses [&NiagaraMeshRendererProperties]
  renderers [{rendererClass=&NiagaraMeshRendererProperties, rendererIndex=0}]
 }
 FE_Deploy_First {
  emitterUpdateScript [
   &EmitterState(Life Cycle Mode=::NewEnumerator1, Loop Behavior=::NewEnumerator1, Loop Duration=0.3, UseLoopDelay=true)
   &SpawnRate(
    SpawnGroup:NiagaraInt32=&RandomRangeInt(Minimum=0, Maximum=1)
    SpawnRate=&FN_FrontEndFloatModifier(Gameplay=0, Front End=160)
    Use Spawn Probability=false
   )
  ]
  particleSpawnScript [
   &FN_SetSpawnGroup(Number of Spawn Groups=2)
   /SetVariables() {isSetParameters=true, module=SetVariables_#3}
   /SetVariables(
    Particles.FrontToBack:NiagaraFloat=&RandomRangeFloat(Minimum=0, Maximum=1)
    Particles.LeanLerp:NiagaraFloat=&Multiply_Float(
     A=$User.RightDot
     B=&Lerp_Float(A=-1, B=1, Alpha=&MakeFloatFromInt(Int=$Particles.SpawnGroup))
    )
   ) {isSetParameters=true, module=SetVariables_#4}
   &InitializeParticle(
    Lifetime Max=0.75
    Lifetime Min=0.75
    Lifetime Mode=::NewEnumerator1
    Mesh Scale Max=<1 2 1.2>
    Mesh Scale Min=<0.8 1.5 1>
    Mesh Scale Mode=::NewEnumerator2
    Sprite Rotation Mode=::NewEnumerator2
    UsePositionOffset=false
    Write Material Random=true
   ) {module=InitializeParticle001}
   &InitialMeshOrientation(
    Orientation Vector=&Multiply_Vector(
     A=&Add_Vector(
      A=&Lerp_Vector(A=<0 4 0>, B=<-2 0 0>, Alpha=$Particles.FrontToBack)
      B=&Lerp_Vector(A=<-1 0.8 0>, B=<-0.75 1 0>, Alpha=$Particles.LeanLerp)
     )
     B=&Multiply_Vector(
      A=&Lerp_Vector(A=<1 1 1>, B=<1 -1 1>, Alpha=&MakeFloatFromInt(Int=$Particles.SpawnGroup))
      B=&Lerp_Vector(A=<1 1 1>, B=<1 2 1>, Alpha=&ClampFloat(Float=$User.ForwardDot))
     )
    )
    Rotation Coordinate Space=::NewEnumerator0
    Use Orientation Vector=true
    Use Rotation Vector=false
   )
   /SetVariables(Particles.Scale=&Multiply_Vector(
     A=&Multiply_Vector(
      A=&VectorFromCurve(
       VectorCurve=DataInterface() {propertyValues {
         XCurve {keys [{arriveTangent=1.7846, interpMode=RCIM_Cubic, leaveTangent=1.7846, tangentMode=RCTM_User, time=0, value=0.4}, <at 1 1 RCIM_Cubic>]}
         YCurve {keys [{arriveTangent=1.7846, interpMode=RCIM_Cubic, leaveTangent=1.7846, tangentMode=RCTM_User, time=0, value=0.4}, <at 1 1 RCIM_Cubic>]}
         ZCurve {keys [{arriveTangent=1.7846, interpMode=RCIM_Cubic, leaveTangent=1.7846, tangentMode=RCTM_User, time=0, value=0.4}, <at 1 1 RCIM_Cubic>]}
        }}
       CurveIndex=$Particles.FrontToBack
       Scale Curve=&Multiply_VectorByFloat(
        Vector=$Particles.Scale
        Float=&Lerp_Float(
         A=0.8
         B=0.8
         Alpha=&ClampFloat(Float=&Max_Float(A:NiagaraFloat=$User.ForwardDot, B:NiagaraFloat=$System.Timeline004.NormalizedTime))
        )
       )
      )
      B=&Lerp_Vector(A=<0.8 0.8 0.8>, B=<0.8 0.8 0.8>, Alpha=$Particles.LeanLerp)
     )
     B=&MakeVector(
      X=&Lerp_Float(
       A=1
       B=1.5
       Alpha=&FloatFromCurve(
        FloatCurve=DataInterface() {propertyValues {Curve {
           defaultValue=0
           keys [
            {arriveTangent=8.0076, interpMode=RCIM_Cubic, leaveTangent=8.0076, tangentMode=RCTM_User, time=0.6, value=0}
            {interpMode=RCIM_Cubic, tangentMode=RCTM_User, time=1, value=1}
           ]
          }}}
        CurveIndex=$Particles.FrontToBack
       )
      )
      Y=&Lerp_Float(A=2, B=1, Alpha=$Particles.FrontToBack)
      Z=1
     )
    )) {isSetParameters=true, module=SetVariables_#5}
  ]
  particleUpdateScript [
   &ParticleState()
   /SetVariables(Particles.Scale=&Multiply_Vector(
     A=$Particles.Initial.Scale
     B=&VectorFromCurve(VectorCurve=DataInterface() {propertyValues {
        XCurve {keys [
          {arriveTangent=4.6994, interpMode=RCIM_Cubic, leaveTangent=4.6994, tangentMode=RCTM_User, time=0, value=0.0795}
          {arriveTangent=0.4589, interpMode=RCIM_Cubic, leaveTangent=0.4589, tangentMode=RCTM_User, time=0.1, value=0.7609}
          <at 0.5 1>
         ]}
        YCurve {keys [{arriveTangent=2.5596, interpMode=RCIM_Cubic, leaveTangent=2.5596, tangentMode=RCTM_User, time=0, value=0.0767}, <at 0.5 1>]}
        ZCurve {keys [
          <at 0 0 RCIM_Cubic>
          {arriveTangent=-2.3619, interpMode=RCIM_Cubic, leaveTangent=-2.3619, tangentMode=RCTM_User, time=0.15, value=0.9567}
          <at 0.9988 0>
         ]}
       }})
    )) {isSetParameters=true, module=SetVariables_#6}
   &SkeletalMeshLocation(
    Bone / Socket Index=$Particles.SpawnGroup
    "Inherit Velocity (Scale)"=-1
    MaskSpawnGroup=false
    Mesh Sampling Type=::NewEnumerator5
    Sampled Position Offset=&VectorFromCurve(
     VectorCurve=DataInterface() {propertyValues {
       XCurve {keys [<at 0 50>, {arriveTangent=-382.5781, interpMode=RCIM_Cubic, leaveTangent=-382.5781, tangentMode=RCTM_User, time=1, value=-190}]}
       YCurve {keys [<at 0.33 0 RCIM_Cubic>, {arriveTangent=-92.7914, interpMode=RCIM_Cubic, leaveTangent=-92.7914, tangentMode=RCTM_User, time=1, value=-35}]}
       ZCurve {keys [<at 0 0>,<at 1 10>]}
      }}
     CurveIndex=$Particles.FrontToBack
     Scale Curve=&Lerp_Vector(A=<1 1 1>, B=<1 -1 1>, Alpha=&MakeFloatFromInt(Int=$Particles.SpawnGroup))
    )
    Skeletal Mesh=$User.SkeletalMesh
    Socket Sampling Mode=::NewEnumerator5
    Velocity Sampling=::NewEnumerator0
    VelocitySpeedLimit=false
   ) {module=SkeletalMeshLocation004}
   &SolveForcesAndVelocity(Clamp Velocity=false, Limit Acceleration=false)
   &DynamicMaterialParameters(
    Index 0 Param 1=&FloatFromCurve(FloatCurve=DataInterface() {propertyValues {Curve {keys [
         {arriveTangent=0.0095, interpMode=RCIM_Cubic, leaveTangent=0.0095, tangentMode=RCTM_User, time=0, value=-1.5}
         {arriveTangent=0.0695, interpMode=RCIM_Cubic, leaveTangent=0.0695, tangentMode=RCTM_User, time=1, value=0}
        ]}}})
    Index 0 Param 2=&FloatFromCurve(FloatCurve=DataInterface() {propertyValues {Curve {keys [<at 0 4>,<at 1 4>]}}})
    Index 0 Param 3=&MakeFloatFromVector(Vector=$Particles.Scale, Channel=::NewEnumerator0)
    Index 0 Param 4=1
    Index 1 Param 1=&FloatFromCurve(FloatCurve=DataInterface() {propertyValues {Curve {keys [<at 0.7 0 RCIM_Cubic>, {arriveTangent=6.4641, interpMode=RCIM_Cubic, leaveTangent=6.4641, tangentMode=RCTM_User, time=1, value=1}]}}})
    Module.Param0WriteEnabled=true
    Module.Param0WriteEnabled1=true
    Module.Param1WriteEnabled=true
    Module.Param1WriteEnabled1=true
    Module.Param2WriteEnabled=true
    Module.Param2WriteEnabled1=true
    Module.Param3WriteEnabled=true
    Module.Param3WriteEnabled1=true
    Write Parameter Index 1=true
   )
   &ScaleColor(
    Scale Alpha=&FloatFromCurve(FloatCurve=DataInterface() {propertyValues {Curve {keys [{interpMode=RCIM_Cubic, tangentMode=RCTM_User, time=0, value=1}, {interpMode=RCIM_Cubic, tangentMode=RCTM_User, time=0.16, value=1}]}}})
    ScaleA=true
    ScaleRGB=true
   )
  ]
  rendererClasses [&NiagaraMeshRendererProperties]
  renderers [{rendererClass=&NiagaraMeshRendererProperties, rendererIndex=0}]
 }
 FE_Deploy_Second {
  emitterUpdateScript [
   &EmitterState(Life Cycle Mode=::NewEnumerator1, Loop Behavior=::NewEnumerator1, Loop Delay=0.4, Loop Duration=0.6, UseLoopDelay=true)
   &SpawnRate(
    SpawnGroup:NiagaraInt32=&RandomRangeInt(Minimum=0, Maximum=1)
    SpawnRate=&FN_FrontEndFloatModifier(Gameplay=0, Front End=160)
    Use Spawn Probability=false
   )
  ]
  particleSpawnScript [
   /SetVariables(Particles.SpawnGroup=1) {isSetParameters=true, module=SetVariables_#3}
   /SetVariables(
    Particles.FrontToBack:NiagaraFloat=&RandomRangeFloat(Minimum=0, Maximum=1)
    Particles.LeanLerp:NiagaraFloat=&Multiply_Float(
     A=$User.RightDot
     B=&Lerp_Float(A=-1, B=1, Alpha=&MakeFloatFromInt(Int=$Particles.SpawnGroup))
    )
   ) {isSetParameters=true, module=SetVariables_#4}
   &InitializeParticle(
    Lifetime Max=0.75
    Lifetime Min=0.75
    Lifetime Mode=::NewEnumerator1
    Mesh Scale Max=<1.5 2 1.2>
    Mesh Scale Min=<1 1.5 1>
    Mesh Scale Mode=::NewEnumerator2
    Sprite Rotation Mode=::NewEnumerator2
    UsePositionOffset=false
    Write Material Random=true
   ) {module=InitializeParticle001}
   &InitialMeshOrientation(
    Orientation Vector=&Multiply_Vector(
     A=&Add_Vector(
      A=&Lerp_Vector(A=<0 4 0>, B=<-2 0 0>, Alpha=$Particles.FrontToBack)
      B=&Lerp_Vector(A=<-1 0.8 0>, B=<-0.75 1 0>, Alpha=$Particles.LeanLerp)
     )
     B=&Multiply_Vector(
      A=&Lerp_Vector(A=<1 1 1>, B=<1 -1 1>, Alpha=&MakeFloatFromInt(Int=$Particles.SpawnGroup))
      B=&Lerp_Vector(A=<1 1 1>, B=<1 2 1>, Alpha=&ClampFloat(Float=$User.ForwardDot))
     )
    )
    Rotation Coordinate Space=::NewEnumerator0
    Use Orientation Vector=true
    Use Rotation Vector=false
   )
   /SetVariables(Particles.Scale=&Multiply_Vector(
     A=&Multiply_Vector(
      A=&VectorFromCurve(
       VectorCurve=DataInterface() {propertyValues {
         XCurve {keys [{arriveTangent=1.7846, interpMode=RCIM_Cubic, leaveTangent=1.7846, tangentMode=RCTM_User, time=0, value=0.4}, <at 1 1 RCIM_Cubic>]}
         YCurve {keys [{arriveTangent=1.7846, interpMode=RCIM_Cubic, leaveTangent=1.7846, tangentMode=RCTM_User, time=0, value=0.4}, <at 1 1 RCIM_Cubic>]}
         ZCurve {keys [{arriveTangent=1.7846, interpMode=RCIM_Cubic, leaveTangent=1.7846, tangentMode=RCTM_User, time=0, value=0.4}, <at 1 1 RCIM_Cubic>]}
        }}
       CurveIndex=$Particles.FrontToBack
       Scale Curve=&Multiply_VectorByFloat(
        Vector=$Particles.Scale
        Float=&Lerp_Float(
         A=0.8
         B=0.8
         Alpha=&ClampFloat(Float=&Max_Float(A:NiagaraFloat=$User.ForwardDot, B:NiagaraFloat=$System.Timeline004.NormalizedTime))
        )
       )
      )
      B=&Lerp_Vector(A=<0.8 0.8 0.8>, B=<0.8 0.8 0.8>, Alpha=$Particles.LeanLerp)
     )
     B=&MakeVector(
      X=&Lerp_Float(
       A=1
       B=1.5
       Alpha=&FloatFromCurve(
        FloatCurve=DataInterface() {propertyValues {Curve {
           defaultValue=0
           keys [
            {arriveTangent=8.0076, interpMode=RCIM_Cubic, leaveTangent=8.0076, tangentMode=RCTM_User, time=0.6, value=0}
            {interpMode=RCIM_Cubic, tangentMode=RCTM_User, time=1, value=1}
           ]
          }}}
        CurveIndex=$Particles.FrontToBack
       )
      )
      Y=&Lerp_Float(A=2, B=1, Alpha=$Particles.FrontToBack)
      Z=1
     )
    )) {isSetParameters=true, module=SetVariables_#5}
   &AddVelocity(Velocity=<-1 0.25 1>, Velocity Speed Scale=500)
  ]
  particleUpdateScript [
   &ParticleState()
   /SetVariables(Particles.Scale=&Multiply_Vector(
     A=$Particles.Initial.Scale
     B=&VectorFromCurve(VectorCurve=DataInterface() {propertyValues {
        XCurve {keys [
          {arriveTangent=4.6994, interpMode=RCIM_Cubic, leaveTangent=4.6994, tangentMode=RCTM_User, time=0, value=0.0795}
          {arriveTangent=0.4589, interpMode=RCIM_Cubic, leaveTangent=0.4589, tangentMode=RCTM_User, time=0.1, value=0.7609}
          <at 0.5 1>
         ]}
        YCurve {keys [{arriveTangent=2.5596, interpMode=RCIM_Cubic, leaveTangent=2.5596, tangentMode=RCTM_User, time=0, value=0.0767}, <at 0.5 1>]}
        ZCurve {keys [
          <at 0 0 RCIM_Cubic>
          {arriveTangent=-2.3619, interpMode=RCIM_Cubic, leaveTangent=-2.3619, tangentMode=RCTM_User, time=0.15, value=0.9567}
          <at 0.9988 0>
         ]}
       }})
    )) {isSetParameters=true, module=SetVariables_#6}
   &SkeletalMeshLocation(
    Bone / Socket Index=$Particles.SpawnGroup
    MaskSpawnGroup=false
    Mesh Sampling Type=::NewEnumerator5
    Sampled Position Offset=&VectorFromCurve(
     VectorCurve=DataInterface() {propertyValues {
       XCurve {keys [<at 0 50>, {arriveTangent=-382.5781, interpMode=RCIM_Cubic, leaveTangent=-382.5781, tangentMode=RCTM_User, time=1, value=-190}]}
       YCurve {keys [<at 0.33 -10 RCIM_Cubic>, {arriveTangent=-92.7914, interpMode=RCIM_Cubic, leaveTangent=-92.7914, tangentMode=RCTM_User, time=1, value=-45}]}
       ZCurve {keys [<at 0 0>,<at 1 10>]}
      }}
     CurveIndex=$Particles.FrontToBack
     Scale Curve=&Lerp_Vector(A=<1 1 1>, B=<1 -1 1>, Alpha=&MakeFloatFromInt(Int=$Particles.SpawnGroup))
    )
    Skeletal Mesh=$User.SkeletalMesh
    Socket Sampling Mode=::NewEnumerator5
   ) {module=SkeletalMeshLocation004}
   &SolveForcesAndVelocity(Clamp Velocity=false, Limit Acceleration=false)
   &DynamicMaterialParameters(
    Index 0 Param 1=&FloatFromCurve(FloatCurve=DataInterface() {propertyValues {Curve {keys [
         {arriveTangent=0.0095, interpMode=RCIM_Cubic, leaveTangent=0.0095, tangentMode=RCTM_User, time=0, value=-1.5}
         {arriveTangent=0.0695, interpMode=RCIM_Cubic, leaveTangent=0.0695, tangentMode=RCTM_User, time=1, value=0}
        ]}}})
    Index 0 Param 2=&FloatFromCurve(FloatCurve=DataInterface() {propertyValues {Curve {keys [<at 0 4>,<at 1 4>]}}})
    Index 0 Param 3=&MakeFloatFromVector(Vector=$Particles.Scale, Channel=::NewEnumerator0)
    Index 0 Param 4=1
    Index 1 Param 1=&FloatFromCurve(FloatCurve=DataInterface() {propertyValues {Curve {keys [<at 0.7 0 RCIM_Cubic>, {arriveTangent=6.4641, interpMode=RCIM_Cubic, leaveTangent=6.4641, tangentMode=RCTM_User, time=1, value=1}]}}})
    Module.Param0WriteEnabled=true
    Module.Param0WriteEnabled1=true
    Module.Param1WriteEnabled=true
    Module.Param1WriteEnabled1=true
    Module.Param2WriteEnabled=true
    Module.Param2WriteEnabled1=true
    Module.Param3WriteEnabled=true
    Module.Param3WriteEnabled1=true
    Write Parameter Index 1=true
   )
   &ScaleColor(
    Scale Alpha=&FloatFromCurve(FloatCurve=DataInterface() {propertyValues {Curve {keys [{interpMode=RCIM_Cubic, tangentMode=RCTM_User, time=0, value=1}, {interpMode=RCIM_Cubic, tangentMode=RCTM_User, time=0.16, value=1}]}}})
    ScaleA=true
    ScaleRGB=true
   )
  ]
  rendererClasses [&NiagaraMeshRendererProperties]
  renderers [{rendererClass=&NiagaraMeshRendererProperties, rendererIndex=0}]
 }
}
summary {
 systemName=NS_Glider_FloralMane_Deploy
 emitters [
  {emitterName=Deploy_First, rendererClasses [&NiagaraMeshRendererProperties]}
  {emitterName=Deploy_Second, rendererClasses [&NiagaraMeshRendererProperties]}
  {emitterName=FE_Deploy_First, rendererClasses [&NiagaraMeshRendererProperties]}
  {emitterName=FE_Deploy_Second, rendererClasses [&NiagaraMeshRendererProperties]}
 ]
 userVariables [
  User.bIsFrontEnd {type=NiagaraBool}
  User.bIsFrontEndPreview {type=NiagaraBool}
  User.bIsFullyDeployed {type=NiagaraBool}
  User.ForwardDot {type=NiagaraFloat, defaultValue=0, description=""}
  User.RightDot {type=NiagaraFloat, defaultValue=0, description=""}
  User.SkeletalMesh {
   type=NiagaraDataInterfaceSkeletalMesh
   defaultValue {
    struct=NiagaraExt_VariableValue_DataInterface
    value {dataInterface=&self:NiagaraDataInterfaceSkeletalMesh_1}
   }
  }
 ]
}
