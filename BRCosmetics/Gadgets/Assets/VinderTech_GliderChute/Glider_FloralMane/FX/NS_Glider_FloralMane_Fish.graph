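# Niagara effect for the FloralMane glider: mesh fish leap in short gravity arcs from offsets either side of the glider's skeletal mesh.
# Fish spawn rate falls toward zero as the larger of ForwardDot and a system timeline rises.
# Splash, Mist and Ripples sprites spawn from fish via particle reads, with parent-age kill curves confining them to brief windows near each leap's start and end.
# Time-of-day functions adjust fish material, mist and ripples at night.
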NiagaraSystem /BRCosmetics/Gadgets/Assets/VinderTech_GliderChute/Glider_FloralMane/FX/NS_Glider_FloralMane_Fish
dependencies {
 usedDataInterfaces [
  &NiagaraDataInterfaceArrayFloat3
  &NiagaraDataInterfaceCurve
  &NiagaraDataInterfaceArrayFloat
  &NiagaraDataInterfaceParticleRead
  &NiagaraDataInterfaceSpriteRendererInfo
 ]
 usedDynamicInputs [&Boolean_OrOperation, &SetBoolByFloatComparison]
 usedModules [
  &self:SystemSpawnScript.SystemScriptSource.SystemScriptGraph.NiagaraNodeAssignment_1.SetVariables_#1
  &SystemState
  &FN_Integrator
  &self:SystemSpawnScript.SystemScriptSource.SystemScriptGraph.NiagaraNodeAssignment_0.SetVariables_#2
 ]
}
emitters {
 Mist {
  emitterUpdateScript [
   &EmitterState(Life Cycle Mode=::NewEnumerator1, Loop Behavior=::NewEnumerator1, Loop Delay=3, Loop Duration Mode=::NewEnumerator1, UseLoopDelay=true)
   &SpawnParticlesFromOtherEmitter(
    Attribute Reader=DataInterface() {propertyValues {EmitterBinding {bindingMode=Other, emitterName=fish}}}
    Spawn Rate=100
    Spawn Rate Per Particle Cap=200
    Use Maximum Age=false
    Use Minimum Age=false
   )
  ]
  particleSpawnScript [
   &InitializeParticle(
    Lifetime Max=0.5
    Lifetime Min=0.35
    Lifetime Mode=::NewEnumerator1
    Mass Max=2
    Mass Mode=::NewEnumerator3
    Sprite Size Mode=::NewEnumerator3
    Uniform Sprite Size=&RandomRangeFloat(Minimum=25, Maximum=40)
    UsePositionOffset=false
    Write Material Random=true
   )
   &SampleParticlesFromOtherEmitter()
   &self:GetParentParticleAge(
    Index:NiagaraInt32=&MakeIntFromFloat(Float=$Particles.SampleParticlesFromOtherEmitter.SampledIndex)
    ParentEmitter=DataInterface() {propertyValues {EmitterBinding {bindingMode=Other, emitterName=fish}}}
   ) {module=GetParentParticleAge}
   &SubUVAnimation(
    Evaluation Type=::NewEnumerator0
    "Sprite Renderer "=DataInterface() {propertyValues {SpriteRenderer=&self:mist_0.NiagaraSpriteRendererProperties_1}}
    Start Frame Range Override=20
    SubUV Animation Mode:ENiagara_SubUVLookupModeV2=::NewEnumerator2
    UseEndFrame=true
    UseStartFrame=true
   ) {module=SubUVAnimation001}
   &ShapeLocation(MaskSpawnGroup=false, Override Seed=false, Ring Radius=5, Shape Primitive=::NewEnumerator4, UseOwnerScale=true)
   &AddVelocity(
    Module.Use Falloff Distance=false
    Origin Offset=<0 0 -20>
    Override Seed=false
    Velocity Mode=::NewEnumerator1
    Velocity Origin=$Output.SampleParticlesFromOtherEmitter.SampledPosition
    Velocity Speed=250
   ) {module=AddVelocity002}
  ]
  particleUpdateScript [
   &ParticleState()
   &ScaleSpriteSize(Uniform Curve Sprite Scale=DataInterface() {propertyValues {Curve {
       defaultValue=0
       keys [{interpMode=RCIM_Cubic, tangentMode=RCTM_User, time=0, value=1}, {interpMode=RCIM_Cubic, tangentMode=RCTM_User, time=1, value=2}]
      }}}) {module=ScaleSpriteSize001}
   &SpriteRotationRate(Rotation Rate=&RandomRangeFloat(Minimum=-90, Maximum=90))
   &self:KillParticlesBasedOnParentParticleAge(
    KillCurve:NiagaraInt32=&MakeIntFromFloat(Float=&FloatFromCurve(
      FloatCurve=DataInterface() {propertyValues {Curve {keys [<at 0 1 RCIM_Constant>,<at 0.05 0 RCIM_Constant>,<at 0.1 1 RCIM_Constant>,<at 0.85 0 RCIM_Constant>,<at 0.9 1 RCIM_Constant>]}}}
      CurveIndex=$Particles.Initial.ParentNormalizedLife
     ))
    isFirstFrame:NiagaraBool=$Output.ParticleState.FirstFrame
   ) {module=KillParticlesBasedOnParentParticleAge001}
   &GravityForce()
   &Drag(Drag=&RandomRangeFloat(Minimum=0.8, Maximum=1.2), Use Linear Drag=true, Use Rotational Drag=false)
   &SolveForcesAndVelocity(Clamp Velocity=false, Limit Acceleration=false)
   &ScaleColor(
    Scale Alpha=&FloatFromCurve(
     FloatCurve=DataInterface() {propertyValues {Curve {keys [
         {interpMode=RCIM_Cubic, tangentMode=RCTM_User, time=0, value=0}
         {arriveTangent=0.0458, interpMode=RCIM_Cubic, leaveTangent=0.0458, tangentMode=RCTM_User, time=0.1, value=1}
         {interpMode=RCIM_Cubic, tangentMode=RCTM_User, time=0.6, value=1}
         {interpMode=RCIM_Cubic, tangentMode=RCTM_User, time=1, value=0}
        ]}}}
     Scale Curve=0.05
    )
    Scale RGB=&FN_Select_Vector_from_Time_Of_Day(Morning=<1 1 1>, Day=<1 1 1>, Evening=<1 1 1>, Night=<0.25 0.25 0.8>)
    ScaleA=true
    ScaleRGB=true
   ) {module=ScaleColor001}
   &DynamicMaterialParameters(
    Index 0 Param 1=&RandomRangeFloat(Minimum=0, Maximum=1)
    Index 0 Param 3=&FloatFromCurve(FloatCurve=DataInterface() {propertyValues {Curve {keys [{interpMode=RCIM_Cubic, tangentMode=RCTM_User, time=0, value=0}, {interpMode=RCIM_Cubic, tangentMode=RCTM_User, time=1, value=1}]}}})
    Index 0 Param 4=&FloatFromCurve(FloatCurve=DataInterface() {propertyValues {Curve {keys [<at 0 0 RCIM_Cubic>, {arriveTangent=1.5276, interpMode=RCIM_Cubic, leaveTangent=1.5276, tangentMode=RCTM_User, time=1, value=1}]}}})
    Module.Param0WriteEnabled=true
    Module.Param1WriteEnabled=false
    Module.Param2WriteEnabled=true
    Module.Param3WriteEnabled=true
   )
  ]
  rendererClasses [&NiagaraSpriteRendererProperties]
  renderers [{rendererClass=&NiagaraSpriteRendererProperties}]
 }
 Ripples {
  emitterUpdateScript [
   &EmitterState(Life Cycle Mode=::NewEnumerator1, Loop Behavior=::NewEnumerator1, Loop Delay=3, Loop Duration Mode=::NewEnumerator1, UseLoopDelay=true)
   &SpawnParticlesFromOtherEmitter(
    Attribute Reader=DataInterface() {propertyValues {EmitterBinding {bindingMode=Other, emitterName=fish}}}
    Spawn Rate=30
    Spawn Rate Per Particle Cap=200
    Use Maximum Age=false
    Use Minimum Age=false
   )
  ]
  particleSpawnScript [
   &InitializeParticle(
    Lifetime=&Lerp_Float(A=&RandomRangeFloat(Minimum=0.35, Maximum=0.5), B=0.5, Alpha=0.5)
    Mass Max=2
    Mass Mode=::NewEnumerator3
    Sprite Size Mode=::NewEnumerator3
    Uniform Sprite Size=32
    UsePositionOffset=false
    Write Material Random=true
   )
   &SampleParticlesFromOtherEmitter()
   &self:GetParentParticleAge(
    Index:NiagaraInt32=&MakeIntFromFloat(Float=$Particles.SampleParticlesFromOtherEmitter.SampledIndex)
    ParentEmitter=DataInterface() {propertyValues {EmitterBinding {bindingMode=Other, emitterName=fish}}}
   ) {module=GetParentParticleAge}
  ]
  particleUpdateScript [
   &ParticleState()
   &self:KillParticlesBasedOnParentParticleAge(
    KillCurve:NiagaraInt32=&MakeIntFromFloat(Float=&FloatFromCurve(
      FloatCurve=DataInterface() {propertyValues {Curve {keys [<at 0 1 RCIM_Constant>,<at 0.05 0 RCIM_Constant>,<at 0.1 1 RCIM_Constant>,<at 0.85 0 RCIM_Constant>,<at 0.9 1 RCIM_Constant>]}}}
      CurveIndex=$Particles.Initial.ParentNormalizedLife
     ))
    isFirstFrame:NiagaraBool=$Output.ParticleState.FirstFrame
   ) {module=KillParticlesBasedOnParentParticleAge}
   &Drag(Drag=&RandomRangeFloat(Minimum=0.8, Maximum=1.2), Use Linear Drag=true, Use Rotational Drag=false)
   &ScaleSpriteSize(Uniform Curve Sprite Scale=DataInterface() {propertyValues {Curve {
       defaultValue=0
       keys [
        {arriveTangent=2.7552, interpMode=RCIM_Cubic, leaveTangent=2.7552, tangentMode=RCTM_User, time=0, value=0}
        {interpMode=RCIM_Cubic, tangentMode=RCTM_User, time=1, value=2}
       ]
      }}}) {module=ScaleSpriteSize001}
   &SolveForcesAndVelocity(Clamp Velocity=false, Limit Acceleration=false)
   &ScaleColor(
    Scale Alpha=&FloatFromCurve(
     FloatCurve=DataInterface() {propertyValues {Curve {keys [
         {interpMode=RCIM_Cubic, tangentMode=RCTM_User, time=0, value=0}
         {arriveTangent=0.0458, interpMode=RCIM_Cubic, leaveTangent=0.0458, tangentMode=RCTM_User, time=0.1, value=1}
         {interpMode=RCIM_Cubic, tangentMode=RCTM_User, time=0.4, value=1}
         {interpMode=RCIM_Cubic, tangentMode=RCTM_User, time=0.8, value=0}
        ]}}}
     Scale Curve=0.25
    )
    Scale RGB=&FN_Select_Vector_from_Time_Of_Day(Morning=<1 1 1>, Day=<1 1 1>, Evening=<1 1 1>, Night=<0.2 0.2 1>)
    ScaleA=true
    ScaleRGB=true
   ) {module=ScaleColor001}
   &SpriteFacingAndAlignment(Facing Coordinate Space=::World, Sprite Facing=<0 0 1>, WriteAlignment=false, WriteFacing=true)
   /SetVariables(Particles.Sprite2:Vector2f=&Multiply_Vector2D_ByFloat(
     Float=&FloatFromCurve(
      FloatCurve=DataInterface() {propertyValues {Curve {keys [<at 0.5 0.25>,<at 1 1>]}}}
      Scale Curve=&RandomRangeFloat(Minimum=0.75, Maximum=1)
     )
     Vector2D=$Particles.SpriteSize
    )) {isSetParameters=true}
   /SetVariables(Particles.sprite2color:LinearColor=&MultiplyLinearColor(
     Linear Color=<rgba 1 1 1 1>
     Scale Factor=&MakeVector4FromVectorAndFloat(
      "Vector (RGB)"=&FN_Select_Vector_from_Time_Of_Day(Morning=<1 1 1>, Day=<1 1 1>, Evening=<1 1 1>, Night=<0.2 0.2 1>)
      "Float (Alpha)"=&FloatFromCurve(
       FloatCurve=DataInterface() {propertyValues {Curve {keys [
           {interpMode=RCIM_Cubic, tangentMode=RCTM_User, time=0.2, value=0}
           {arriveTangent=0.0458, interpMode=RCIM_Cubic, leaveTangent=0.0458, tangentMode=RCTM_User, time=0.3, value=1}
           {interpMode=RCIM_Cubic, tangentMode=RCTM_User, time=0.6, value=1}
           {interpMode=RCIM_Cubic, tangentMode=RCTM_User, time=1, value=0}
          ]}}}
       Scale Curve=0.25
      )
     )
    )) {isSetParameters=true}
   /SetVariables(Particles.sprite2rotate:NiagaraFloat=&RandomRangeFloat(Minimum=0, Maximum=360)) {isSetParameters=true}
  ]
  rendererClasses [&NiagaraSpriteRendererProperties]
  renderers [
   {rendererClass=&NiagaraSpriteRendererProperties}
   {rendererClass=&NiagaraSpriteRendererProperties, rendererIndex=1}
  ]
 }
 Splash {
  emitterUpdateScript [
   &EmitterState(Life Cycle Mode=::NewEnumerator1, Loop Behavior=::NewEnumerator1, Loop Delay=3, Loop Duration Mode=::NewEnumerator1, UseLoopDelay=true)
   &SpawnParticlesFromOtherEmitter(
    Attribute Reader=DataInterface() {propertyValues {EmitterBinding {bindingMode=Other, emitterName=fish}}}
    Spawn Rate=200
    Spawn Rate Per Particle Cap=200
    Use Maximum Age=false
    Use Minimum Age=false
   )
  ]
  particleSpawnScript [
   &InitializeParticle(
    Lifetime Max=0.5
    Lifetime Min=0.35
    Lifetime Mode=::NewEnumerator1
    Mass Max=2
    Mass Mode=::NewEnumerator3
    Sprite Size Mode=::NewEnumerator3
    Uniform Sprite Size=&RandomRangeFloat(Minimum=20, Maximum=35)
    UsePositionOffset=false
    Write Material Random=true
   )
   &SampleParticlesFromOtherEmitter()
   &self:GetParentParticleAge(
    Index:NiagaraInt32=&MakeIntFromFloat(Float=$Particles.SampleParticlesFromOtherEmitter.SampledIndex)
    ParentEmitter=DataInterface() {propertyValues {EmitterBinding {bindingMode=Other, emitterName=fish}}}
   ) {module=GetParentParticleAge}
   &SubUVAnimation(
    End Frame Range Override=3
    Evaluation Type=::NewEnumerator0
    "Sprite Renderer "=DataInterface() {propertyValues {SpriteRenderer=&self:splash_0.NiagaraSpriteRendererProperties_1}}
    SubUV Animation Mode:ENiagara_SubUVLookupModeV2=::NewEnumerator2
    UseEndFrame=true
    UseStartFrame=true
   ) {module=SubUVAnimation001}
   &ShapeLocation(MaskSpawnGroup=false, Override Seed=false, Ring Radius=5, Shape Primitive=::NewEnumerator4, UseOwnerScale=true)
   &AddVelocity(
    Module.Use Falloff Distance=false
    Origin Offset=<0 0 -20>
    Override Seed=false
    Velocity Mode=::NewEnumerator1
    Velocity Origin=$Output.SampleParticlesFromOtherEmitter.SampledPosition
    Velocity Speed=250
   ) {module=AddVelocity002}
  ]
  particleUpdateScript [
   &ParticleState()
   &self:KillParticlesBasedOnParentParticleAge(
    KillCurve:NiagaraInt32=&MakeIntFromFloat(Float=&FloatFromCurve(
      FloatCurve=DataInterface() {propertyValues {Curve {keys [<at 0 1 RCIM_Constant>,<at 0.05 0 RCIM_Constant>,<at 0.1 1 RCIM_Constant>,<at 0.85 0 RCIM_Constant>,<at 0.9 1 RCIM_Constant>]}}}
      CurveIndex=$Particles.Initial.ParentNormalizedLife
     ))
    isFirstFrame:NiagaraBool=$Output.ParticleState.FirstFrame
   ) {module=KillParticlesBasedOnParentParticleAge}
   &SpriteRotationRate(Rotation Rate=&RandomRangeFloat(Minimum=-90, Maximum=90))
   &GravityForce()
   &Drag(Drag=&RandomRangeFloat(Minimum=0.8, Maximum=1.2), Use Linear Drag=true, Use Rotational Drag=false)
   &SolveForcesAndVelocity(Clamp Velocity=false, Limit Acceleration=false)
   &ScaleColor(
    Scale Alpha=&FloatFromCurve(FloatCurve=DataInterface() {propertyValues {Curve {keys [
         {interpMode=RCIM_Cubic, tangentMode=RCTM_User, time=0, value=0}
         {arriveTangent=0.0458, interpMode=RCIM_Cubic, leaveTangent=0.0458, tangentMode=RCTM_User, time=0.33, value=1}
         {interpMode=RCIM_Cubic, tangentMode=RCTM_User, time=0.6, value=1}
         {interpMode=RCIM_Cubic, tangentMode=RCTM_User, time=1, value=0}
        ]}}})
    ScaleA=true
    ScaleRGB=true
   ) {module=ScaleColor001}
   &DynamicMaterialParameters(
    Index 0 Param 1=&RandomRangeFloat(Minimum=0, Maximum=1)
    Index 0 Param 3=&FloatFromCurve(FloatCurve=DataInterface() {propertyValues {Curve {keys [{interpMode=RCIM_Cubic, tangentMode=RCTM_User, time=0, value=0}, {interpMode=RCIM_Cubic, tangentMode=RCTM_User, time=1, value=1}]}}})
    Index 0 Param 4=&FloatFromCurve(FloatCurve=DataInterface() {propertyValues {Curve {keys [<at 0 0.6 RCIM_Cubic>, {arriveTangent=0.6991, interpMode=RCIM_Cubic, leaveTangent=0.6991, tangentMode=RCTM_User, time=1, value=1}]}}})
    Module.Param0WriteEnabled=true
    Module.Param1WriteEnabled=false
    Module.Param2WriteEnabled=true
    Module.Param3WriteEnabled=true
   )
  ]
  rendererClasses [&NiagaraSpriteRendererProperties]
  renderers [{rendererClass=&NiagaraSpriteRendererProperties}]
 }
 fish {
  emitterUpdateScript [
   &EmitterState(
    Inactive Response=::NewEnumerator1
    Life Cycle Mode=::NewEnumerator1
    Loop Behavior=::NewEnumerator1
    Loop Delay=3
    Loop Duration Mode=::NewEnumerator1
    UseLoopDelay=true
   )
   &SpawnRate(
    SpawnGroup:NiagaraInt32=&RandomRangeInt(Minimum=0, Maximum=1)
    SpawnRate=&Lerp_Float(
     A=3
     B=0
     Alpha=&ClampFloat(Float=&Max_Float(A:NiagaraFloat=$User.ForwardDot, B:NiagaraFloat=$System.Timeline004.NormalizedTime))
    )
    Use Spawn Probability=false
   )
  ]
  particleSpawnScript [
   /SetVariables(
    Particles.FrontToBack:NiagaraFloat=&RandomRangeFloat(Minimum=0, Maximum=1)
    Particles.LeanLerp:NiagaraFloat=&Multiply_Float(
     A=$User.RightDot
     B=&Lerp_Float(A=-1, B=1, Alpha=&MakeFloatFromInt(Int=$Particles.SpawnGroup))
    )
   ) {isSetParameters=true}
   &InitializeParticle(
    Lifetime Max=0.8
    Lifetime Min=0.6
    Lifetime Mode=::NewEnumerator1
    Mesh Scale Mode=::NewEnumerator3
    Mesh Uniform Scale=&RandomRangeFloat(Minimum=0.8, Maximum=1)
    Sprite Rotation Mode=::NewEnumerator2
    UsePositionOffset=false
    Write Material Random=true
   ) {module=InitializeParticle001}
   &SkeletalMeshLocation(
    Apply Sampled Orientation to Mesh Particles=true
    Bone / Socket Index=2
    Inherit Velocity=false
    "Inherit Velocity (Scale)"=0
    MaskSpawnGroup=false
    Mesh Sampling Type=::NewEnumerator5
    Sampled Position Offset=&SelectVectorFromArray(
     Array Sampling Mode=::NewEnumerator0
     Direct Array Index=?
     Vector Selection Array=DataInterface() {propertyValues {FloatData [<-250 125 0>, <-250 -125 0>]}}
     Random Seed=0
    )
    Sampled Position Offset Coordinate Space:ENiagaraMeshTransforms=::NewEnumerator3
    Skeletal Mesh=$User.SkeletalMesh
    Socket Sampling Mode=::NewEnumerator5
   )
   &ShapeLocation(Disc Coverage=1, MaskSpawnGroup=false, Override Seed=false, Ring Radius=60, Shape Primitive=::NewEnumerator4, UseOwnerScale=true)
   &AddVelocity(
    Velocity=&MakeVector(
     X=&RandomRangeFloat(Minimum=-0.1, Maximum=0.1)
     Y=&RandomRangeFloat(Minimum=-0.5, Maximum=-1)
     Z=0.8
    )
    Velocity Speed Scale=&RandomRangeFloat(Minimum=300, Maximum=400)
   ) {module=AddVelocity001}
  ]
  particleUpdateScript [
   &ParticleState()
   &ScaleMeshSize(Scale Factor=&VectorFromFloat(Value=&FloatFromCurve(FloatCurve=DataInterface() {propertyValues {Curve {
         defaultValue=1
         keys [
          {interpMode=RCIM_Cubic, tangentMode=RCTM_User, time=0, value=0}
          {arriveTangentWeight=0.1067, interpMode=RCIM_Cubic, leaveTangentWeight=0.3202, tangentMode=RCTM_User, time=0.2, value=1}
          {arriveTangentWeight=0.2, interpMode=RCIM_Cubic, leaveTangentWeight=0.0667, tangentMode=RCTM_User, time=0.8, value=1}
          {tangentMode=RCTM_User, time=1, value=0}
         ]
        }}})))
   &GravityForce()
   &SolveForcesAndVelocity(Clamp Velocity=false, Limit Acceleration=false)
   &InterpolateOverTime(Rate Of Change=8, Target Vector Value:Vector3f=<0 0 1>, Variable Type=::NewEnumerator2) {module=InterpolateOverTime001}
   &InterpolateOverTime(Rate Of Change=8) {module=InterpolateOverTime002}
   &Timeline(
    Play Rate=&Multiply_Float(A=1, B=0.005)
    Rewind Mode:ENiagara_TimelineMode=::NewEnumerator1
    Timeline End=10000000
   )
   &DynamicMaterialParameters(
    Index 0 Param 1=&Add_Float(A=$Particles.Timeline.Time, B=&RandomRangeFloat(Minimum=0, Maximum=50))
    Index 0 Param 2=&RandomRangeFloat(Minimum=0, Maximum=1)
    Index 0 Param 3=&Add_Float(
     A=&Multiply_Float(A=$Particles.InterpolateOverTime002.InterpolateOverTime_Float.Moving Average, B=10)
     B=&Waveform(
      Global Input Position=&RandomRangeFloat(Minimum=0, Maximum=500)
      "Waveform [1]"=::NewEnumerator12
      "[1] Frequency"=0.4
      "[1] Amplitude Min/Max"=<0.4 0.6>
     )
    )
    Index 0 Param 4=&MaskFloatBySpawnGroup(
     Masked Float=&SelectFloatFromArray(
      Direct Array Index=&MakeIntFromFloat(
       Float=&FloatFromCurve(
        FloatCurve=DataInterface() {propertyValues {Curve {keys [<at 0 0 RCIM_Constant>,<at 0.45 1 RCIM_Constant>,<at 0.6 2 RCIM_Constant>]}}}
        CurveIndex=&ReturnNormalizedExecIndex()
       )
       Conversion Mode=::NewEnumerator1
      )
      Direct Array Mode=::NewEnumerator1
      Float Selection Array=DataInterface() {propertyValues {FloatData [-0.3,0.4,0.4]}}
     )
     Passthrough Float=0.05
    )
    Index 1 Param 1=&MaskFloatBySpawnGroup(Masked Float=&RandomRangeFloat(Minimum=0.4, Maximum=0.6), Passthrough Float=0.8)
    Index 1 Param 2=&FloatFromCurve(
     FloatCurve=DataInterface() {propertyValues {Curve {keys [<at 0 1>,<at 0.4 1>,<at 0.5 0.1>,<at 1 0.1>]}}}
     CurveIndex=&RandomRangeFloat(Minimum=0, Maximum=1)
     Scale Curve=&FNTimeOfDayScalar(Night Scale=0.1, Morning Scale=1, Evening Scale=1)
    )
    Index 3 Param 1=&MakeFloatFromVector(Vector=$Particles.InterpolateOverTime001.InterpolateOverTime_V3_.Moving Average, Channel=::NewEnumerator0)
    Index 3 Param 2=&MakeFloatFromVector(Vector=$Particles.InterpolateOverTime001.InterpolateOverTime_V3_.Moving Average, Channel=::NewEnumerator1)
    Index 3 Param 3=&MakeFloatFromVector(Vector=$Particles.InterpolateOverTime001.InterpolateOverTime_V3_.Moving Average, Channel=::NewEnumerator2)
    Index 3 Param 4=$Particles.InterpolateOverTime002.InterpolateOverTime_Float.Moving Average
    Module.Index0Param0Write=true
    Module.Index0Param1Write=true
    Module.Index0Param2Write=true
    Module.Index0Param3Write=true
    Module.Index1Param0Write=true
    Module.Index1Param1Write=true
    Module.Index1Param2Write=true
    Module.Index1Param3Write=true
    Module.Index3Param0Write=true
    Module.Index3Param1Write=true
    Module.Index3Param2Write=true
    Module.Index3Param3Write=true
    Write Parameter Index 1=true
    Write Parameter Index 3=true
   ) {module=DynamicMaterialParameters001}
  ]
  rendererClasses [&NiagaraMeshRendererProperties]
  renderers [{rendererClass=&NiagaraMeshRendererProperties, rendererIndex=0}]
 }
}
summary {
 systemName=NS_Glider_FloralMane_Fish
 emitters [
  {emitterName=fish, rendererClasses [&NiagaraMeshRendererProperties]}
  {emitterName=Splash, rendererClasses [&NiagaraSpriteRendererProperties]}
  {emitterName=Mist, rendererClasses [&NiagaraSpriteRendererProperties]}
  {emitterName=Ripples, rendererClasses [&NiagaraSpriteRendererProperties]}
 ]
 userVariables [
  User.bIsFrontEnd {type=NiagaraBool}
  User.bIsFrontEndPreview {type=NiagaraBool}
  User.bIsFullyDeployed {type=NiagaraBool}
  User.ForwardDot {type=NiagaraFloat, defaultValue=0, description=""}
  User.RightDot {type=NiagaraFloat, defaultValue=0, description=""}
  User.SkeletalMesh {
   type=NiagaraDataInterfaceSkeletalMesh
   defaultValue {
    struct=NiagaraExt_VariableValue_DataInterface
    value {dataInterface=&self:NiagaraDataInterfaceSkeletalMesh_1}
   }
  }
 ]
}
